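# Physics asset for the Baba Yaga backpack cosmetic, setting up physics bodies for its dangling lock, ribbon, ropes and rings.
# Box bodies on dyn_ bones form separate constraint groups, each rooted at a tiny base body.
# The ribbon forks into a two-segment branch and a single segment, the large rope has two segments, and the other pieces have one.
# Tight swing and twist limits restrain motion, with the lock restricted to nearly single-axis swing.
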PhysicsAsset /BRCosmetics/Accessories/FORT_Backpacks/Backpack_F_MED_Baba_Yaga/Meshes/F_MED_Baba_Yaga_Pack_Physics
bodies {
 dyn_lock_base [{shapeType=Box, extentX=0.5, extentY=0.5, extentZ=0.5}]
 dyn_lock_parent [{shapeType=Box, extentX=4.6013, extentY=2.5, extentZ=7.8312, center=<2.0055 0.0001 0>}]
 dyn_ribbon_01 [{shapeType=Box, extentX=3.2586, extentY=2, extentZ=0.83, center=<1.6011 0 0.0001>}]
 dyn_ribbon_02 [{shapeType=Box, extentX=4.4492, extentY=2.1871, extentZ=0.8927, center=<1.8016 0 1.2682>, rotation=<rot 37.4635 -0.0001 0>}]
 dyn_ribbon_03 [{shapeType=Box, extentX=4.052, extentY=2.0476, extentZ=1.035, center=<1.9746 0 0>}]
 dyn_ribbon_base [{shapeType=Box, extentX=0.5, extentY=0.5, extentZ=0.5}]
 dyn_ribbon_parent [{shapeType=Box, extentX=3.01, extentY=2.01, extentZ=0.76, center=<1.6048 0 0>}]
 dyn_ring_base_01 [{shapeType=Box, extentX=0.5, extentY=0.5, extentZ=0.5}]
 dyn_ring_base_02 [{shapeType=Box, extentX=0.5, extentY=0.5, extentZ=0.5}]
 dyn_ring_parent_01 [{shapeType=Box, extentX=3.6738, extentY=1.0613, extentZ=3.3108, center=<0 0 1.7052>}]
 dyn_ring_parent_02 [{shapeType=Box, extentX=2.4887, extentY=0.7567, extentZ=2.7037, center=<0 -0.0766 1.3965>}]
 dyn_rope_01 [{shapeType=Box, extentX=6.0706, extentY=6.5797, extentZ=1.1046, center=<2.9527 0 0>}]
 dyn_rope_base [{shapeType=Box, extentX=0.5, extentY=0.5, extentZ=0.5}]
 dyn_rope_parent [{shapeType=Box, extentX=6.3468, extentY=7.1366, extentZ=1.4458, center=<3.1196 0 0>}]
 dyn_rope_small_base [{shapeType=Box, extentX=0.5004, extentY=0.5004, extentZ=0.5004}]
 dyn_rope_small_parent [{shapeType=Box, extentX=2.8749, extentY=5.7925, extentZ=1.2374, center=<1.5498 0 0>}]
}
constraints [
 {bone1Name=dyn_lock_parent, bone2Name=dyn_lock_base, swing1LimitDegrees=25, swing2LimitDegrees=0, twistLimitDegrees=2}
 {bone1Name=dyn_ribbon_parent, bone2Name=dyn_ribbon_base, swing1LimitDegrees=3, swing2LimitDegrees=15, twistLimitDegrees=2}
 {bone1Name=dyn_ribbon_01, bone2Name=dyn_ribbon_parent, swing1LimitDegrees=10, swing2LimitDegrees=35, twistLimitDegrees=3}
 {bone1Name=dyn_ribbon_02, bone2Name=dyn_ribbon_01, swing1LimitDegrees=20, swing2LimitDegrees=30, twistLimitDegrees=3}
 {bone1Name=dyn_ribbon_03, bone2Name=dyn_ribbon_parent, swing1LimitDegrees=20, swing2LimitDegrees=20, twistLimitDegrees=2}
 {bone1Name=dyn_rope_01, bone2Name=dyn_rope_parent, swing1LimitDegrees=10, swing2LimitDegrees=30, twistLimitDegrees=10}
 {bone1Name=dyn_rope_parent, bone2Name=dyn_rope_base, swing1LimitDegrees=10, swing2LimitDegrees=35, twistLimitDegrees=10}
 {bone1Name=dyn_ring_parent_01, bone2Name=dyn_ring_base_01, swing1LimitDegrees=6, swing2LimitDegrees=6, twistLimitDegrees=2}
 {bone1Name=dyn_ring_parent_02, bone2Name=dyn_ring_base_02, swing1LimitDegrees=6, swing2LimitDegrees=6, twistLimitDegrees=2}
 {bone1Name=dyn_rope_small_parent, bone2Name=dyn_rope_small_base, swing1LimitDegrees=20, swing2LimitDegrees=45, twistLimitDegrees=3}
]
