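# Animation Blueprint for the Baba Yaga backpack cosmetic, animating its bird bones with idle, airborne and landing motion.
# A state machine switches Idle, InAir and JumpLand on bIsFalling, returning to Idle automatically as the landing clip ends.
# Its pose is layered over the input pose from dyn_bird_hips down, then fed through RigidBody physics in root-bone space with damped component motion.
# The event graph holds no logic, and IsOnLandFirst is unused.
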AnimBlueprint /BRCosmetics/Accessories/FORT_Backpacks/Backpack_F_MED_Baba_Yaga/Meshes/F_MED_Baba_Yaga_Pack_AnimBP
target_skeleton=&/F_MED_Baba_Yaga_Pack_Skeleton
graphs [
 AnimGraph {nodes [
   n1 = Input Pose()
   AnimGraphNode_StateMachine_0 = New State Machine() {is=title AnimGraphNode_StateMachine}
   n2 = RigidBody(ComponentPose=Local To Component(LocalPose=Layered blend per bone(BasePose=@n1, BlendPoses_0=@AnimGraphNode_StateMachine_0) {layerSetup [{branchFilters [{boneName=dyn_bird_hips}]}]})) {
    bClampLinearTranslationLimitToRefPose=true
    bOverrideWorldGravity=true
    overridePhysicsAsset=&/F_MED_Baba_Yaga_Pack_Physics
    simulationSpace=BaseBoneSpace
    alphaBoolBlend {blendOption=Linear}
    baseBoneRef {boneName=root}
    componentAppliedLinearAccClamp=<3000 3000 3000>
    componentLinearAccScale=<0.25 0.3 0.35>
    componentLinearVelScale=<0.25 0.3 0.35>
    overrideWorldGravity {z=-980}
   }
   Output Pose(Result=Component To Local(ComponentPose=@n2)) {layerGroup=None}
  ]}
 New State Machine {
  entry=Idle
  path=AnimGraph.AnimGraphNode_StateMachine_0.New State Machine
  ids=id
  nodes [
   AnimStateNode_0 = Idle()
   AnimStateNode_1 = InAir() {is=title AnimStateNode}
   AnimStateNode_2 = JumpLand() {is=title AnimStateNode}
   AnimStateTransitionNode_0 = AnimStateTransitionNode() {crossfadeDuration=0.15}
   AnimStateTransitionNode_1 = AnimStateTransitionNode() {crossfadeDuration=0}
   AnimStateTransitionNode_2 = AnimStateTransitionNode() {bAutomaticRuleBasedOnSequencePlayerInState=true, crossfadeDuration=0.1}
  ]
  links [
   {
    dst=InAir
    rule=AnimGraph.AnimGraphNode_StateMachine_0.New State Machine.AnimStateTransitionNode_0.Transition
    src=Idle
    via=AnimStateTransitionNode_0
   }
   {
    dst=JumpLand
    rule=AnimGraph.AnimGraphNode_StateMachine_0.New State Machine.AnimStateTransitionNode_1.Transition
    src=InAir
    via=AnimStateTransitionNode_1
   }
   {
    dst=Idle
    rule=AnimGraph.AnimGraphNode_StateMachine_0.New State Machine.AnimStateTransitionNode_2.Transition
    src=JumpLand
    via=AnimStateTransitionNode_2
   }
  ]
 }
 Idle {
  path=AnimGraph.AnimGraphNode_StateMachine_0.New State Machine.AnimStateNode_0.Idle
  nodes [Output Animation Pose(Result=Sequence Player 'BabaYaga_Idle'() {sequence=/BRCosmetics/Accessories/FORT_Backpacks/Backpack_F_MED_Baba_Yaga/Animation/BabaYaga_Idle})]
 }
 InAir {
  path=AnimGraph.AnimGraphNode_StateMachine_0.New State Machine.AnimStateNode_1.InAir
  nodes [Output Animation Pose(Result=Sequence Player 'BabaYaga_InAir'() {sequence=/BRCosmetics/Accessories/FORT_Backpacks/Backpack_F_MED_Baba_Yaga/Animation/BabaYaga_InAir})]
 }
 JumpLand {
  path=AnimGraph.AnimGraphNode_StateMachine_0.New State Machine.AnimStateNode_2.JumpLand
  nodes [Output Animation Pose(Result=Sequence Player 'BabaYaga_Jump'() {bLoopAnimation=false, sequence=/BRCosmetics/Accessories/FORT_Backpacks/Backpack_F_MED_Baba_Yaga/Animation/BabaYaga_Jump})]
 }
 Transition {
  path=AnimGraph.AnimGraphNode_StateMachine_0.New State Machine.AnimStateTransitionNode_0.Transition
  nodes [Result(bCanEnterTransition=Get bIsFalling())]
 }
 Transition {
  path=AnimGraph.AnimGraphNode_StateMachine_0.New State Machine.AnimStateTransitionNode_1.Transition
  nodes [Result(bCanEnterTransition=NOT Boolean(A=Get bIsFalling()))]
 }
 Transition {path=AnimGraph.AnimGraphNode_StateMachine_0.New State Machine.AnimStateTransitionNode_2.Transition, nodes [Result()]}
 EventGraph {nodes [TryGetPawnOwner()]}
]
variables [IsOnLandFirst {type=boolean, default=true}]
